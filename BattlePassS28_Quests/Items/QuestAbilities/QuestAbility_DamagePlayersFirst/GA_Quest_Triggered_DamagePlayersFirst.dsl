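; Triggered quest ability that credits the owning player for damaging enemy players before those players damage them.
; Damage-received events add the attacker's name to InvalidPlayerNames.
; On damage dealt to another team, a non-empty list is searched for the target's name; if absent, the name is added and a custom verb is broadcast, then the ability ends.
; An empty list skips the check and broadcast entirely.
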(event Ability|EventActivateAbilityFromEvent
  (params (EventData :type Gameplay%20Event%20Data%20Structure%20%28by%20ref%29))
  (bind (_eventtag _instigator _target _optionalobject _optionalobject2 _contexthandle _instigatortags _targettags) (Utilities|Struct|BreakGameplayEventData EventData))
  (bind _returnvalue (Fortnite|GetPlayerNameSafe _target))
  (bind _invalidplayernames (Variables|Default|GetInvalidPlayerNames))
  (|CallParentFunction EventData)
  ;[1 1 1] Check if a player initiated damage and took damage, if false end ability
  (if (and (GameplayTags|HasTag _instigatortags (Variables|Default|GetTag_Player)) (GameplayTags|HasTag _targettags (Variables|Default|GetTag_Player)))
    (switch Utilities|FlowControl|Switch|SwitchonGameplayTag _eventtag
      (:Default
        _)
      (:Event.Damage.Dealt
        (if (not (Team|OnSameTeam _instigator _target))
          ;[1 1 1] If not empty check invalid player list. If empty add damaged actor to list and send verb
          (Variables|Default|SetTookDMGFirst false)
          (if (Utilities|Array|IsNotEmpty (Variables|Default|GetInvalidPlayerNames))
            ;[1 1 1] Check if damaged actor has already been hit or damaged owning player
            (bind _array_element (Utilities|Array|ForEachLoopwithBreak (Variables|Default|GetInvalidPlayerNames))
              (:LoopBody
                (if (Utilities|String|EqualExactly(String) _array_element _returnvalue)
                  (Variables|Default|SetTookDMGFirst true)))
              (:Completed
                _))
            (if (not (Variables|Default|GetTookDMGFirst))
              ;[1 1 1] Add damaged actors to filter out players already hit
              (Utilities|Array|AddUnique _invalidplayernames _returnvalue)
              ;[1 1 1] Send Custom Verb
              (CustomVerb|BroadcastCustomVerbMessage (CustomVerb|MakeCustomVerbMessage (Variables|GAVars|Data|GetOwningPlayerPawn) (Variables|Default|GetVerb%20Name)))
              (else
                _))
            (Ability|EndAbility))
          (else
            _)))
      (:Event.Damage.Dealt.Overshield
        _)
      (:Event.Damage.Dealt.Shield
        _)
      (:Event.Damage.Received
        ;[1 1 1] Add actors that damaged owning player to filter them out
        (Utilities|Array|AddUnique _invalidplayernames (Fortnite|GetPlayerNameSafe _instigator)))
      (:Event.Damage.Shielded
        _)
      (:Event.Damage.Overshielded
        _))
    (else
      _)))
